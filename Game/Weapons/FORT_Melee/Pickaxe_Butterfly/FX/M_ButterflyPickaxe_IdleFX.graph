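# Translucent unlit, two-sided mesh-particle material that draws animated butterflies for the Butterfly pickaxe idle effect.
# Color multiplies the pickaxe texture, with optional hue shift, desaturation and particle color, by a per-particle three-color blend.
# Two UV-channel masks select wings, which sine-driven RotateAboutAxis flaps around the particle or object pivot with vertical bob, timed by dynamic parameters.
# Alpha combines dynamic-parameter control with a sphere-mask fade gated by vertical object-to-actor distance.
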Material /Game/Weapons/FORT_Melee/Pickaxe_Butterfly/FX/M_ButterflyPickaxe_IdleFX
graph {
 properties {
  bUsedWithMeshParticles=true
  bUsedWithNiagaraMeshParticles=true
  blendMode=BLEND_Translucent
  pixelDepthOffsetMode=PDOM_Legacy
  shadingModel=MSM_Unlit
  twoSided=true
  usedShadingModels=Unlit
 }
 nodes [
  n1 = VectorParameter() {parameterName=EmissiveColorA, defaultValue=<rgba 1 1 1 0>}
  n2 = VectorParameter() {parameterName=EmissiveColorB, defaultValue=<rgba 1 1 1 0>}
  n3 = DynamicParameter() {parameterIndex=1, defaultValue=<rgba 1 1 1 0>, paramNames [RandomTimeOffset, HueColorHandler, RandomColorLerp, InitialRotationOffset]}
  LinearInterpolate(A=@n1.RGB, B=@n2.RGB, Alpha=@n3.RandomColorLerp)
  n4 = VectorParameter() {parameterName=EmissiveColorC, defaultValue=<rgba 1 1 1 0>}
  n5 = TextureSampleParameter2D(UVs=TextureCoordinate()) {parameterName=BaseTex, texture=/Game/Weapons/FORT_Melee/Pickaxe_Butterfly/Textures/T_Pickaxe_Butterfly_D}
  n6 = DynamicParameter() {parameterIndex=1, defaultValue=<rgba 1 1 1 0>, paramNames [RandomTimeOffset, HueColorHandler, RandomColorLerp, InitialRotationOffset]}
  n7 = Desaturation(
   None=LinearInterpolate(
    A=@n5.RGB
    B=MaterialFunctionCall(Hue Shift Percentage=@n6.HueColorHandler, Texture=@n5.RGB) {materialFunction=&HueShift}.Result
    Alpha=ScalarParameter() {parameterName=UseHueColorHandler}
   )
   Fraction=ScalarParameter() {defaultValue=-2, parameterName=Desaturation}
  )
  n8 = LinearInterpolate(A=Constant() {r=1}, B=ParticleColor().RGB, Alpha=ScalarParameter() {parameterName=Use particle color})
  MaterialFunctionCall(Emissive=Multiply(
    A=MaterialFunctionCall(A=@n1.RGB, B=@n2.RGB, C=@n4.RGB, Alpha=@n3.RandomColorLerp) {materialFunction=&3ColorBlend}.Result
    B=Multiply(A=@n7, B=@n8)
   )) {materialFunction=&MF_FX_EmissiveAdjustments}
  n9 = ScalarParameter() {parameterName=Turn off Transparent the paritcle Base on the distance of emitter position to actor position}
  n10 = Subtract(A=ObjectPositionWS(), B=ActorPositionWS())
  n11 = Constant3Vector() {constant=<rgba 0 0 1 0>}
  n12 = SphereMask(
   A=WorldPosition().XYZ
   B=LinearInterpolate(A=ActorPositionWS(), B=ObjectPositionWS(), Alpha=ScalarParameter() {parameterName=Use Object Position})
   Radius=ScalarParameter() {defaultValue=110, parameterName=FadeRadius}
   Hardness=ScalarParameter() {parameterName=FadeStrength}
  )
  n13 = If(
   A=MaterialFunctionCall(Vector 3=Multiply(A=@n10, B=@n11)) {materialFunction=/Engine/Functions/Engine_MaterialFunctions02/Utility/VectorLength}.V3 Length
   B=Constant() {r=105}
   "A > B"=@n12
   "A == B"=@n12
   "A < B"=Constant() {r=1}
  )
  n14 = Constant() {r=1}
  n15 = DynamicParameter() {paramNames [FreqencyWing, Control Alpha, "MotionRadius(0-00)", UpDownAmount]}
  n16 = ScalarParameter() {parameterName=Use Dynamic to Control Alpha}
  Reroute(None=Multiply(A=LinearInterpolate(A=@n13, Alpha=@n9), B=LinearInterpolate(A=@n14, B=@n15.Control Alpha, Alpha=@n16)))
  n17 = TextureCoordinate() {coordinateIndex=1}
  n18 = Constant()
  n19 = VectorParameter() {parameterName=RotationAxis, defaultValue=<rgba 0 1 0 1>}
  n20 = DynamicParameter() {parameterIndex=1, defaultValue=<rgba 1 1 1 0>, paramNames [RandomTimeOffset, HueColorHandler, RandomColorLerp, InitialRotationOffset]}
  n21 = Add(A=Time(), B=@n20.RandomTimeOffset) {constB=0}
  n22 = DynamicParameter() {paramNames [FreqencyWing, Control Alpha, "MotionRadius(0-00)", UpDownAmount]}
  n23 = Multiply(A=Sine(None=Multiply(A=@n21, B=@n22.FreqencyWing))) {constB=0.25}
  n24 = ScalarParameter() {defaultValue=0.12}
  n25 = Add(A=@n23, B=@n24) {constB=0.06}
  n26 = ParticlePositionWS()
  n27 = ObjectPositionWS()
  n28 = ScalarParameter() {parameterName=Use Object Position}
  n29 = Multiply(
   A=If(A=ComponentMask(None=@n17) {r=true}, B=Constant() {r=0.01}, "A > B"=Constant() {r=1}, "A == B"=@n18, "A < B"=@n18)
   B=RotateAboutAxis(
    NormalizedRotationAxis=Transform(None=@n19.RGB) {transformSourceType=TRANSFORMSOURCE_Local}
    RotationAngle=Add(A=@n25, B=@n20.InitialRotationOffset)
    PivotPoint=LinearInterpolate(A=@n26, B=@n27, Alpha=@n28)
    Position=WorldPosition().XYZ
   )
  )
  n30 = VectorParameter() {parameterName=RotationAxis, defaultValue=<rgba 0 1 0 1>}
  n31 = Time()
  n32 = Subtract(A=Multiply(A=Add(A=@n20.RandomTimeOffset, B=@n31), B=@n22.FreqencyWing)) {constB=0.5}
  n33 = Add(A=Multiply(A=Sine(None=@n32)) {constB=0.25}, B=Multiply(A=@n24) {constB=-1}) {constB=-0.15}
  n34 = ParticlePositionWS()
  n35 = ObjectPositionWS()
  n36 = ScalarParameter() {parameterName=Use Object Position}
  n37 = TextureCoordinate() {coordinateIndex=2}
  n38 = Constant()
  n39 = Multiply(
   A=RotateAboutAxis(
    NormalizedRotationAxis=Transform(None=@n30.RGB) {transformSourceType=TRANSFORMSOURCE_Local}
    RotationAngle=Add(A=@n20.InitialRotationOffset, B=@n33)
    PivotPoint=LinearInterpolate(A=@n34, B=@n35, Alpha=@n36)
    Position=WorldPosition().XYZ
   )
   B=If(A=ComponentMask(None=@n37) {r=true}, B=Constant() {r=0.02}, "A > B"=Constant() {r=1}, "A == B"=@n38, "A < B"=@n38)
  )
  n40 = Multiply(A=@n23)
  n41 = Transform(None=Multiply(
    A=DynamicParameter() {paramNames [FreqencyWing, Control Alpha, "MotionRadius(0-00)", UpDownAmount]}.UpDownAmount
    B=Constant3Vector() {constant=<rgba 0 0 1 0>}
   )) {transformSourceType=TRANSFORMSOURCE_Local}
  Reroute(None=Add(A=Add(A=@n29, B=@n39), B=Multiply(A=@n40, B=@n41)) {constB=0})
  ScalarParameter() {parameterName=Hue}
 ]
}
